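# Physics asset for the large male Buff Cat backpack cosmetic, defining simulation bodies for its dangling mousenades, yarn and strings.
# Capsule chains hang off small box "master" bodies via angle-limited constraints, while the pack sphere has no constraint.
# Each mousenade parent carries knot strings and a two-body ring with asymmetric swing limits for hinge-like motion; the left also carries a yarn loop.
# Yarn and long-string chains run separately from their own masters.
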PhysicsAsset /BRCosmetics/Accessories/FORT_Backpacks/M_LRG_Buff_Cat_Backpack/Meshes/M_LRG_Buff_Cat_Pack_Physics
bodies {
 dyn_knot_string_l [{shapeType=Capsule, length=1.8441, radius=0.7355, center=<1.343 0 0>, rotation=<rot 90 0 0>}]
 dyn_knot_string_r_01 [{shapeType=Capsule, length=1.2471, radius=0.6902, center=<0.9773 0 0>, rotation=<rot 90 0 0>}]
 dyn_knot_string_r_02 [{shapeType=Capsule, length=1.1403, radius=0.8042, center=<1.1789 0 0>, rotation=<rot 90 0 0>}]
 dyn_long_string_r_01 [{shapeType=Capsule, length=2.9984, radius=0.9838, center=<1.9851 0 0>, rotation=<rot 90 0 0>}]
 dyn_long_string_r_02 [{shapeType=Capsule, length=2.9984, radius=0.9838, center=<1.9851 0 0>, rotation=<rot 90 0 0>}]
 dyn_long_string_r_master [{shapeType=Box, extentX=0.5, extentY=0.5, extentZ=0.5}]
 dyn_long_string_r_parent [{shapeType=Capsule, length=2.9984, radius=0.9838, center=<1.9851 0 0>, rotation=<rot 90 0 0>}]
 dyn_mousenade_l_master [{shapeType=Box, extentX=0.5, extentY=0.5, extentZ=0.5}]
 dyn_mousenade_l_parent [{shapeType=Capsule, length=3.5792, radius=1.651, center=<2.2896 0 0>, rotation=<rot 90 0 0>}]
 dyn_mousenade_r_master [{shapeType=Box, extentX=0.5, extentY=0.5, extentZ=0.5}]
 dyn_mousenade_r_parent [{shapeType=Capsule, length=3.5792, radius=1.651, center=<2.6313 0 0>, rotation=<rot 90 0 0>}]
 dyn_mousenade_ring_l [{shapeType=Capsule, length=1, radius=1, shapeName=dyn_mousenade_ring_r_capsule, center=<3 0 0>, rotation=<rot 90 0 0>}]
 dyn_mousenade_ring_l_base [{shapeType=Capsule, length=0.505, radius=0.505, shapeName=dyn_mousenade_ring_r_base_capsule, rotation=<rot 0 -90 0>}]
 dyn_mousenade_ring_r [{shapeType=Capsule, length=1, radius=1, shapeName=dyn_mousenade_ring_r_capsule, center=<3 0 0>, rotation=<rot 90 0 0>}]
 dyn_mousenade_ring_r_base [{shapeType=Capsule, length=0.505, radius=0.505, shapeName=dyn_mousenade_ring_r_base_capsule, rotation=<rot 0 -90 0>}]
 dyn_yarn_loop_l [{shapeType=Capsule, length=1.3437, radius=0.6049, center=<0.9531 0 0>, rotation=<rot 90 0 0>}]
 dyn_yarn_string_l_01 [{shapeType=Capsule, length=4.4315, radius=0.9008, center=<2.4347 0.1811 0.8668>, rotation=<rot 0 94.2535 -70.4525>}]
 dyn_yarn_string_l_master [{shapeType=Box, extentX=0.5, extentY=0.5, extentZ=0.5}]
 dyn_yarn_string_l_parent [{shapeType=Capsule, length=4.4315, radius=0.9008, center=<4.0224 0.1465 1.173>, rotation=<rot 0 94.2535 -70.4525>}]
 dyn_yarn_string_r_master [{shapeType=Box, extentX=0.75, extentY=0.5, extentZ=0.75}]
 dyn_yarn_string_r_parent [{shapeType=Capsule, length=2.0018, radius=0.8724, center=<1.6414 0 0>, rotation=<rot 90 0 0>}]
 pack [{shapeType=Sphere, radius=2}]
}
constraints [
 {bone1Name=dyn_mousenade_l_parent, bone2Name=dyn_mousenade_l_master, swing1LimitDegrees=20, swing2LimitDegrees=20, twistLimitDegrees=5}
 {bone1Name=dyn_mousenade_r_parent, bone2Name=dyn_mousenade_r_master, swing1LimitDegrees=10, swing2LimitDegrees=10, twistLimitDegrees=0}
 {bone1Name=dyn_yarn_string_r_parent, bone2Name=dyn_yarn_string_r_master, swing1LimitDegrees=30, swing2LimitDegrees=30, twistLimitDegrees=10}
 {bone1Name=dyn_yarn_string_l_parent, bone2Name=dyn_yarn_string_l_master, swing1LimitDegrees=30, swing2LimitDegrees=30, twistLimitDegrees=10}
 {bone1Name=dyn_yarn_string_l_01, bone2Name=dyn_yarn_string_l_parent, swing1LimitDegrees=30, swing2LimitDegrees=30, twistLimitDegrees=10}
 {bone1Name=dyn_knot_string_l, bone2Name=dyn_mousenade_l_parent, swing1LimitDegrees=20, swing2LimitDegrees=20, twistLimitDegrees=5}
 {bone1Name=dyn_yarn_loop_l, bone2Name=dyn_mousenade_l_parent, swing1LimitDegrees=10, swing2LimitDegrees=10, twistLimitDegrees=5}
 {bone1Name=dyn_knot_string_r_01, bone2Name=dyn_mousenade_r_parent, swing1LimitDegrees=30, swing2LimitDegrees=30, twistLimitDegrees=10}
 {bone1Name=dyn_knot_string_r_02, bone2Name=dyn_knot_string_r_01, swing1LimitDegrees=30, swing2LimitDegrees=30, twistLimitDegrees=10}
 {bone1Name=dyn_long_string_r_parent, bone2Name=dyn_long_string_r_master, swing1LimitDegrees=20, swing2LimitDegrees=20, twistLimitDegrees=5}
 {bone1Name=dyn_long_string_r_01, bone2Name=dyn_long_string_r_parent, swing1LimitDegrees=20, swing2LimitDegrees=20, twistLimitDegrees=5}
 {bone1Name=dyn_long_string_r_02, bone2Name=dyn_long_string_r_01, swing1LimitDegrees=20, swing2LimitDegrees=20, twistLimitDegrees=5}
 {bone1Name=dyn_mousenade_ring_r, bone2Name=dyn_mousenade_ring_r_base, swing1LimitDegrees=5, swing2LimitDegrees=90, twistLimitDegrees=0}
 {bone1Name=dyn_mousenade_ring_r_base, bone2Name=dyn_mousenade_r_parent, swing1LimitDegrees=90, swing2LimitDegrees=20, twistLimitDegrees=90}
 {bone1Name=dyn_mousenade_ring_l, bone2Name=dyn_mousenade_ring_l_base, swing1LimitDegrees=5, swing2LimitDegrees=90, twistLimitDegrees=0}
 {bone1Name=dyn_mousenade_ring_l_base, bone2Name=dyn_mousenade_l_parent, swing1LimitDegrees=90, swing2LimitDegrees=5, twistLimitDegrees=2}
]
